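; Placement tag test for missions that decides whether a placement actor counts as a complex defense and should receive this test's tags.
; ShouldApplyTagsToPlacementActor runs a shared override-tag macro, returning true when complex is forced and false when it is neither forced nor allowed.
; When complex is only allowed, the shared complex-defense check macro supplies the result.
; The simple and normal flags and the average-location input go unused.
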(fn ShouldApplyTagsToPlacementActor
  (params (FocusPlacementActor :type Fort%20Placement%20Actor%20Object%20Reference) (AveragePlacementActorLocation :type Vector%20%28by%20ref%29 :default "0, 0, 0"))
  (bind (_allow_simple _allow_normal _allow_complex _force_simple _force_normal _force_complex) (PlacementTagTests|PlacementMacroOverrideTagCheck FocusPlacementActor)
    (:Out
      (if _force_complex
        (LocalVariables|SetbIsComplexDefense true)
        (elif _allow_complex
          (bind _return_value (PlacementTagTests|PlacementMacroComplexDefenseCheck FocusPlacementActor)
            (:Out
              (LocalVariables|SetbIsComplexDefense _return_value)))
          (else
            (LocalVariables|SetbIsComplexDefense false))))
      (return (LocalVariables|GetbIsComplexDefense)))))

(fn UserConstructionScript)
