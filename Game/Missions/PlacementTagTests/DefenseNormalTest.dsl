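; Placement tag test for missions that decides whether a placement actor qualifies for normal defense tags.
; An override tag check macro runs first: a force-normal result passes outright, and a lack of allow-normal fails.
; Otherwise the actor counts as normal by elimination, passing only when both the simple and complex defense check macros return false.
; The complex check runs only after the simple check rules the actor out, and the average location input is unused.
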(fn ShouldApplyTagsToPlacementActor
  (params (FocusPlacementActor :type Fort%20Placement%20Actor%20Object%20Reference) (AveragePlacementActorLocation :type Vector%20%28by%20ref%29 :default "0, 0, 0"))
  (bind (_allow_simple _allow_normal _allow_complex _force_simple _force_normal) (PlacementTagTests|PlacementMacroOverrideTagCheck FocusPlacementActor)
    (:Out
      (if _force_normal
        (LocalVariables|SetbIsNormalDefense true)
        (elif _allow_normal
          (bind _return_value (PlacementTagTests|PlacementMacroSimpleDefenseCheck FocusPlacementActor)
            (:Out
              (bind _output_get (LocalVariables|SetbIsNormalDefense (not _return_value)))
              (if _output_get
                (bind _return_value_2 (PlacementTagTests|PlacementMacroComplexDefenseCheck FocusPlacementActor)
                  (:Out
                    (LocalVariables|SetbIsNormalDefense (not _return_value_2))))
                (else
                  _))))
          (else
            (LocalVariables|SetbIsNormalDefense false))))
      (return (LocalVariables|GetbIsNormalDefense)))))

(fn UserConstructionScript)
