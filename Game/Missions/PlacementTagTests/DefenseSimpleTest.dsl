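; Placement tag test that decides whether a Fort placement actor counts as a simple defense.
; An override tag check macro yields allow and force flags for the focus actor.
; A force-simple flag returns true outright; otherwise, when simple is allowed, the simple defense check macro decides, and any other case returns false.
; The average placement location parameter and the normal and complex allow flags go unused.
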(fn ShouldApplyTagsToPlacementActor
  (params (FocusPlacementActor :type Fort%20Placement%20Actor%20Object%20Reference) (AveragePlacementActorLocation :type Vector%20%28by%20ref%29 :default "0, 0, 0"))
  (bind (_allow_simple _allow_normal _allow_complex _force_simple) (PlacementTagTests|PlacementMacroOverrideTagCheck FocusPlacementActor)
    (:Out
      (if _force_simple
        (LocalVariables|SetbIsSimpleDefense true)
        (elif _allow_simple
          (bind _return_value (PlacementTagTests|PlacementMacroSimpleDefenseCheck FocusPlacementActor)
            (:Out
              (LocalVariables|SetbIsSimpleDefense _return_value)))
          (else
            (LocalVariables|SetbIsSimpleDefense false))))
      (return (LocalVariables|GetbIsSimpleDefense)))))

(fn UserConstructionScript)
