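; Placement tag test for missions that decides whether tags, presumably for a normal gate, apply to a given placement actor.
; Its ShouldApplyTagsToPlacementActor function passes when the actor has a Gate or Powersource placement actor tag and has neither the GateComplex nor GateSimple activity tag.
; The check is purely tag-based; the average placement actor location input is unused.
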(fn ShouldApplyTagsToPlacementActor
  (params (FocusPlacementActor :type Fort%20Placement%20Actor%20Object%20Reference) (AveragePlacementActorLocation :type Vector%20%28by%20ref%29 :default "0, 0, 0"))
  (return (and (Utilities|PlacementMacroHasAnyTag FocusPlacementActor (GameplayTags|MakeLiteralGameplayTagContainer "(GameplayTags=((TagName=\"PlacementActor.Gate\"),(TagName=\"PlacementActor.Powersource\")))")) (not (Utilities|PlacementMacroHasAnyTag FocusPlacementActor (GameplayTags|MakeLiteralGameplayTagContainer "(GameplayTags=((TagName=\"Activities.Placement.GateComplex\"),(TagName=\"Activities.Placement.GateSimple\")))"))))))
