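; Placement tag test that decides whether a mission placement actor counts as low elevation and should receive tags.
; The overridden ShouldApplyTagsToPlacementActor compares the focus actor's location against the average placement actor location.
; It returns true when the actor's Z sits more than a fixed threshold below that average.
; The construction script is empty.
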(fn ShouldApplyTagsToPlacementActor
  (params (FocusPlacementActor :type Fort%20Placement%20Actor%20Object%20Reference) (AveragePlacementActorLocation :type Vector%20%28by%20ref%29 :default "0, 0, 0"))
  (return (< (.z (- (Transformation|GetActorLocation FocusPlacementActor) AveragePlacementActorLocation)) -1925)))

(fn UserConstructionScript)
