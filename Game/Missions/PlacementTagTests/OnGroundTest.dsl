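; Placement tag test for mission placement that decides whether a placement actor counts as on the ground, and so should receive its tags.
; ShouldApplyTagsToPlacementActor runs a short vertical line trace spanning just above and below the actor's location, ignoring the actor itself.
; A hit that casts to BuildingSMActor passes unless its building type is Prop or Corner.
; The average placement location input is unused.
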(fn ShouldApplyTagsToPlacementActor
  (params (FocusPlacementActor :type Fort%20Placement%20Actor%20Object%20Reference) (AveragePlacementActorLocation :type Vector%20%28by%20ref%29 :default "0, 0, 0"))
  (bind _returnvalue (Transformation|GetActorLocation FocusPlacementActor))
  (bind _array (Utilities|Array|MakeArray FocusPlacementActor))
  (bind _returnvalue_2 (+ _returnvalue "0, 0,20.000000"))
  (bind _returnvalue_3 (+ _returnvalue "0, 0,-20.000000"))
  (bind (_outhit _returnvalue_4) (PlacementTagTests|LineTracebyChannel FocusPlacementActor _returnvalue_3 _returnvalue_2 "TraceTypeQuery10" false _array))
  (bind (_bblockinghit _binitialoverlap _time _distance _location _impactpoint _normal _impactnormal _physmat _hitactor) (Collision|BreakHitResult _outhit))
  (if _returnvalue_4
    (bind _asbuilding_static_mesh (Utilities|Casting|CastToBuildingSMActor _hitactor)
      (:then
        ;[1 1 1] Don't count props or corners as on ground
        (if (or (Utilities|Enum|Equal(Enum) (Variables|/Script/FortniteGame.BuildingActor|GetBuildingType _asbuilding_static_mesh) "Prop") (Utilities|Enum|Equal(Enum) (Variables|/Script/FortniteGame.BuildingActor|GetBuildingType _asbuilding_static_mesh) "Corner"))
          (LocalVariables|SetbOnGround false)
          (else
            (LocalVariables|SetbOnGround true)))
        (return (LocalVariables|GetbOnGround)))
      (:CastFailed
        _))))

(fn UserConstructionScript)
