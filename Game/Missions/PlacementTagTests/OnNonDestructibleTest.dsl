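; Placement tag test for missions that decides whether tags apply to a placement actor on non-destructible ground.
; ShouldApplyTagsToPlacementActor runs the PlacementMacroonIndestructibleGround macro on the focus placement actor.
; It stores the result in the bOnNonDestructible local and returns it, ignoring the average placement location input.
; The construction script is empty.
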(fn ShouldApplyTagsToPlacementActor
  (params (FocusPlacementActor :type Fort%20Placement%20Actor%20Object%20Reference) (AveragePlacementActorLocation :type Vector%20%28by%20ref%29 :default "0, 0, 0"))
  (bind _return_value (PlacementTagTests|PlacementMacroonIndestructibleGround FocusPlacementActor)
    (:Out
      (bind _output_get (LocalVariables|SetbOnNonDestructible _return_value))
      (return _output_get))))

(fn UserConstructionScript)
